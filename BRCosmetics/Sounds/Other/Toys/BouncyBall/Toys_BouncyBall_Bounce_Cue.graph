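# Sound cue for the bounce impact of the bouncy ball toy.
# A random node picks one of three impact waves with equal weight, then a modulator varies pitch, with a lowered minimum, for variety.
# It plays at reduced volume through a ball-bounce attenuation asset and the Fort_Music_Emote sound class.
# A concurrency override caps simultaneous bounces, stopping the oldest and applying volume scaling.
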SoundCue /BRCosmetics/Sounds/Other/Toys/BouncyBall/Toys_BouncyBall_Bounce_Cue
root=Modulator_0
nodes [
 {id=Modulator_0, children [Random_1], properties {pitchMin=0.85}}
 {id=Random_1, children [WavePlayer_2,WavePlayer_3,WavePlayer_4], properties {weights [1,1,1]}}
 {id=WavePlayer_2, properties {soundWaveAssetPtr=&/Athena_Toys_BouncyBall_Impact_03}}
 {id=WavePlayer_3, properties {soundWaveAssetPtr=&/Athena_Toys_BouncyBall_Impact_01}}
 {id=WavePlayer_4, properties {soundWaveAssetPtr=&/Athena_Toys_BouncyBall_Impact_02}}
]
properties {
 attenuationSettings=/Game/Sounds/Attenuation/Fort_Ball_Bounce_Athena_Att
 bOverrideConcurrency=true
 maxDistance=7000
 soundClassObject=&Fort_Music_Emote
 volumeMultiplier=0.3
 concurrencyOverrides {maxCount=4, resolutionRule=StopOldest, volumeScale=0.9}
}
